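# Animation Blueprint for the infantry rifle skeletal mesh that adds physics-driven secondary motion to its dyn_strap bone.
# The ref pose feeds DefaultSlot for montage playback, then an Anim Dynamics node simulates the strap at partial alpha before output.
# The strap uses a linear spring with reduced gravity, high damping and translation-only limits, with scaled and clamped component motion feeding the simulation.
# The event graph holds no logic.
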AnimBlueprint /Game/Animation/Game/MainPlayer/Combat/Guns/InfantryRifle/SK_InfantryRifle_Dyn_AnimBP
target_skeleton=/Game/Weapons/FORT_Rifles/Mesh/SK_InfantryRifle_Skeleton
graphs [
 AnimGraph {nodes [
   n1 = Local Space Ref Pose()
   n2 = Anim Dynamics - Bone: dyn_strap(ComponentPose=Local To Component(LocalPose=Slot 'DefaultSlot'(Source=@n1)), Alpha=0.25, ExternalForce="0,0,0") {
    bGravityOverrideInSimSpace=true
    bLinearSpring=true
    bOverrideLinearDamping=true
    gravityScale=0.35
    lODThreshold=1
    linearDampingOverride=0.96
    linearSpringConstant=310
    alphaBoolBlend {blendOption=Linear}
    boundBone {boneName=dyn_strap}
    componentAppliedLinearAccClamp=<2000 2000 2000>
    componentLinearAccScale=<0.75 0.75 0.75>
    componentLinearVelScale=<0.75 0.75 0.75>
    physicsBodyDefinitions [{
      sphereCollisionRadius=0
      boundBone {boneName=dyn_strap}
      boxExtents=<15 15 25>
      constraintSetup {
       twistAxis=AxisZ
       angularLimitsMax=<0 0 0>
       angularLimitsMin=<0 0 0>
       angularTarget=<0 0 0>
       linearAxesMax=<10 10 10>
       linearAxesMin=<-10 -10 -10>
      }
      localJointOffset=<0 0 0>
     }]
   }
   Output Pose(Result=Component To Local(ComponentPose=@n2)) {layerGroup=None}
  ]}
 EventGraph {nodes [
   TryGetPawnOwner()
   Event BlueprintUpdateAnimation()
  ]}
]
variables []
